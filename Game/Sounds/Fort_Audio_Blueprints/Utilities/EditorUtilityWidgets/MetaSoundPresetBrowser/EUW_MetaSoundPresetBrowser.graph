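# Editor utility widget that creates a MetaSound Source preset from a chosen MetaSound Source, as an audio authoring tool.
# On Construct, a single property view binds to the widget itself to expose the Preset variable, above an unlabeled button.
# Clicking the button runs inline Python that uses MetaSoundSourceFactory to create an asset referencing that source.
# The new asset takes the source's name plus "_Preset" and lands in the current Content Browser path.
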EditorUtilityWidgetBlueprint /Game/Sounds/Fort_Audio_Blueprints/Utilities/EditorUtilityWidgets/MetaSoundPresetBrowser/EUW_MetaSoundPresetBrowser
graphs [EventGraph {nodes [
   "On Clicked (Button_0)"() {is=title K2Node_ComponentBoundEvent}
   n1 = Get Object Path String(Object=Get Preset() -> (Preset:Meta Sound Source Object Reference))
   > n2 = Execute Python Script!(
    execute
    PythonScript="import unreal\r\n\r\nfactory = unreal.MetaSoundSourceFactory()\r\nfactory.set_editor_property(\"ReferencedMetaSoundObject\", preset)\r\n\r\nmetasound = unreal.AssetToolsHelpers.get_asset_tools().create_asset(name, path, unreal.MetaSoundSource, factory)"
    path:String=GetCurrentContentBrowserPath!() -> (OutPath:String, ReturnValue:bool)
    name:String=Append(A=Split!(InPath:String=@n1) -> (PathPart:String, FilenamePart:String, ExtensionPart:String).FilenamePart, B=_Preset)
    preset:Meta Sound Source Object Reference=Get Preset()
   ) -> (then, ReturnValue, metasound:Object Reference) {inputs [path,name,preset], outputs [metasound]}
   > Cast To MetaSoundSource!(execute, Object=@n2.metasound) -> (then, CastFailed, AsMetaSound Source:Meta Sound Source Object Reference) {is=title K2Node_DynamicCast}
   Event Construct()
   > SetObject(self=Get SinglePropertyView_42() -> (SinglePropertyView_42:Single Property View Object Reference), NewObject=Self-Reference())
   Event PreConstruct()
   Event Tick()
  ]}]
variables [Preset {type=MetaSoundSource, default=null, class=~, properties=~}]
widgets {
 info {rootWidgetClass=&CanvasPanel, widgetCount=4}
 widgets [
  CanvasPanel_0 {widgetClassPath=&CanvasPanel, widget=CanvasPanel_0}
  VerticalBox_0 {
   widgetClassPath=&VerticalBox
   parent=CanvasPanel_0
   widget=VerticalBox_0
   slot {class=CanvasPanelSlot, properties {bAutoSize=true, layoutData=~}}
  }
  SinglePropertyView_42 {
   widgetClassPath=&SinglePropertyView
   parent=VerticalBox_0
   widget=SinglePropertyView_42
   properties {propertyName=Preset}
   slot {class=VerticalBoxSlot}
  }
  Button_0 {
   widgetClassPath=&Button
   parent=VerticalBox_0
   widget=Button_0
   properties {widgetStyle {
     clickedSlateSound {resourceObject=null}
     disabled {
      bIsDynamicallyLoaded=false
      drawAs=RoundedBox
      imageType=NoImage
      mirroring=NoMirror
      resourceName=None
      resourceObject=null
      tiling=NoTile
      imageSize=<0 0>
      margin=<ltrb 0 0 0 0>
      outlineSettings {
       bUseBrushTransparency=false
       roundingType=FixedRadius
       width=1
       color {colorUseRule=UseColor_ColorTable, specifiedColor=<rgba 0 0 0 0>}
       cornerRadii=<4 4 4 4>
      }
      tintColor {colorUseRule=UseColor_Specified, specifiedColor=<rgba 0.0395 0.0395 0.0395 1>}
      uVRegion {bIsValid=false, max=<0 0>, min=<0 0>}
     }
     disabledForeground {colorUseRule=UseColor_ColorTable, specifiedColor=<rgba 0 0 0 0>}
     hovered {
      bIsDynamicallyLoaded=false
      drawAs=RoundedBox
      imageType=NoImage
      mirroring=NoMirror
      resourceName=None
      resourceObject=null
      tiling=NoTile
      imageSize=<32 32>
      margin=<ltrb 0 0 0 0>
      outlineSettings {
       bUseBrushTransparency=false
       roundingType=FixedRadius
       width=1
       color {colorUseRule=UseColor_ColorTable, specifiedColor=<rgba 0 0 0 0>}
       cornerRadii=<4 4 4 4>
      }
      tintColor {colorUseRule=UseColor_Specified, specifiedColor=<rgba 0.0953 0.0953 0.0953 1>}
      uVRegion {bIsValid=false, max=<0 0>, min=<0 0>}
     }
     hoveredForeground {colorUseRule=UseColor_ColorTable, specifiedColor=<rgba 0 0 0 0>}
     hoveredSlateSound {resourceObject=null}
     normal {
      bIsDynamicallyLoaded=false
      drawAs=RoundedBox
      imageType=NoImage
      mirroring=NoMirror
      resourceName=None
      resourceObject=null
      tiling=NoTile
      imageSize=<32 32>
      margin=<ltrb 0 0 0 0>
      outlineSettings {
       bUseBrushTransparency=false
       roundingType=FixedRadius
       width=1
       color {colorUseRule=UseColor_ColorTable, specifiedColor=<rgba 0 0 0 0>}
       cornerRadii=<4 4 4 4>
      }
      tintColor {colorUseRule=UseColor_Specified, specifiedColor=<rgba 0.0395 0.0395 0.0395 1>}
      uVRegion {bIsValid=false, max=<0 0>, min=<0 0>}
     }
     normalForeground {colorUseRule=UseColor_ColorTable, specifiedColor=<rgba 0 0 0 0>}
     normalPadding=<ltrb 12 1.5 12 1.5>
     pressed {
      bIsDynamicallyLoaded=false
      drawAs=RoundedBox
      imageType=NoImage
      mirroring=NoMirror
      resourceName=None
      resourceObject=null
      tiling=NoTile
      imageSize=<32 32>
      margin=<ltrb 0 0 0 0>
      outlineSettings {
       bUseBrushTransparency=false
       roundingType=FixedRadius
       width=1
       color {colorUseRule=UseColor_ColorTable, specifiedColor=<rgba 0 0 0 0>}
       cornerRadii=<4 4 4 4>
      }
      tintColor {colorUseRule=UseColor_Specified, specifiedColor=<rgba 0.0284 0.0284 0.0284 1>}
      uVRegion {bIsValid=false, max=<0 0>, min=<0 0>}
     }
     pressedForeground {colorUseRule=UseColor_ColorTable, specifiedColor=<rgba 0 0 0 0>}
     pressedPadding=<ltrb 12 2.5 12 0.5>
     pressedSlateSound {resourceObject=null}
    }}
   slot {class=VerticalBoxSlot}
  }
 ]
}
